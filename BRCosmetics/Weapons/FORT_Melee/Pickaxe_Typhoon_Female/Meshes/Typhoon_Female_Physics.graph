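# Physics asset for the Typhoon Female pickaxe, a two-handed melee weapon cosmetic with a dangling hanger chain.
# A long capsule covers the handle, and small capsules cover four dynamic hanger joints.
# Constraints link the joints in series from the handle, with a tighter, twist-locked joint at the root and looser 60-degree limits between the links below it.
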PhysicsAsset /BRCosmetics/Weapons/FORT_Melee/Pickaxe_Typhoon_Female/Meshes/Typhoon_Female_Physics
bodies {
 Melee_TwoHanded_Handle [{shapeType=Capsule, length=50, radius=4, center=<0 0 15>}]
 dyn_hanger_jnt_0 [{shapeType=Capsule, length=0.505, radius=0.505, rotation=<rot 0 -90 0>}]
 dyn_hanger_jnt_1 [{shapeType=Capsule, length=3, radius=1, center=<1.3062 0 0>, rotation=<rot 90 0 0>}]
 dyn_hanger_jnt_2 [{shapeType=Capsule, length=3, radius=1, center=<1.3062 0 0>, rotation=<rot 90 0 0>}]
 dyn_hanger_jnt_3 [{shapeType=Capsule, length=3, radius=1, center=<1.3062 0 0>, rotation=<rot 90 0 0>}]
}
constraints [
 {bone1Name=dyn_hanger_jnt_1, bone2Name=dyn_hanger_jnt_0, swing1LimitDegrees=60, swing2LimitDegrees=60, twistLimitDegrees=60}
 {bone1Name=dyn_hanger_jnt_2, bone2Name=dyn_hanger_jnt_1, swing1LimitDegrees=60, swing2LimitDegrees=60, twistLimitDegrees=60}
 {bone1Name=dyn_hanger_jnt_3, bone2Name=dyn_hanger_jnt_2, swing1LimitDegrees=60, swing2LimitDegrees=60, twistLimitDegrees=60}
 {
  bone1Name=dyn_hanger_jnt_0
  bone2Name=Melee_TwoHanded_Handle
  swing1LimitDegrees=20
  swing2LimitDegrees=20
  twistLimitDegrees=10
  twistMotion=Locked
 }
]
